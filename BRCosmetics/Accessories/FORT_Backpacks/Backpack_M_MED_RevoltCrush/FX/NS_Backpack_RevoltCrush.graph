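# Niagara cosmetic effect for the RevoltCrush backpack, emitting drifting mist and small glowing embers.
# Both emitters spawn on indexed planes around the pack; a running flag stops ember spawning and raises the mist rate.
# Embers drift with curl noise, gravity and drag, using a second glow renderer.
# Flipbook mist inherits owner movement, ages faster at high speed, takes time-of-day tint and uses per-plane camera offsets scaled for front end.
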NiagaraSystem /BRCosmetics/Accessories/FORT_Backpacks/Backpack_M_MED_RevoltCrush/FX/NS_Backpack_RevoltCrush
dependencies {
 usedDataInterfaces [
  &NiagaraDataInterfaceCurve
  &NiagaraDataInterfaceArrayFloat3
  &NiagaraDataInterfaceArrayFloat2
  &NiagaraDataInterfaceArrayFloat
  &NiagaraDataInterfaceColorCurve
 ]
 usedDynamicInputs [
  &Vector2DFromFloat
  &RotateVectorByQuaternion
  &MakeVectorFromLinearColorRGB
  &ColorFromCurve
  &Vector4FromCurve
  &FNTimeOfDayScalar
  &SetBoolByFloatComparison
 ]
 usedModules [
  &SystemState
  &self:SystemSpawnScript.SystemScriptSource.SystemScriptGraph.NiagaraNodeAssignment_4.SetVariables_#1
  &self:SystemSpawnScript.SystemScriptSource.SystemScriptGraph.NiagaraNodeAssignment_5.SetVariables_#2
 ]
}
emitters {
 Embers {
  emitterUpdateScript [
   &EmitterState()
   &SpawnRate(
    SpawnRate=&MakeCustomFloatFromBool(Bool=$System.IsRunning, True Float=0, False Float=5)
    Use Spawn Probability=false
   )
  ]
  particleSpawnScript [
   /SetVariables(Particles.PerParticleRandom:NiagaraFloat=&RandomRangeFloat(Minimum=0, Maximum=1)) {isSetParameters=true}
   &InitializeParticle(
    Color=&MultiplyLinearColor(Linear Color=<rgba 1 1 1 0.356>, Scale Factor=&MakeVector4_FromFloat(Float=$System.TOD))
    Lifetime=&RandomRangeFloat(Minimum=0.8, Maximum=1.3)
    Sprite Size Mode=::NewEnumerator3
    Uniform Sprite Size=0.3
    UsePositionOffset=false
    Write Material Random=true
   )
   /SetVariables(Particles.faceindex:NiagaraInt32=&RandomRangeInt(Minimum=0, Maximum=2)) {isSetParameters=true}
   /SetVariables(
    Particles.PlaneAngle:Vector3f=&SelectVectorFromArray(
     Direct Array Index=$Particles.faceindex
     Vector Selection Array=DataInterface() {propertyValues {FloatData [<0 0 90>, <90 0 90>, <90 0 90>]}}
    )
    Particles.PlaneOffset:Vector3f=&SelectVectorFromArray(
     Direct Array Index=$Particles.faceindex
     Vector Selection Array=DataInterface() {propertyValues {FloatData [<0 -12 1>, <-6.5 -6.5 1>, <6.5 -6.5 1>]}}
    )
   ) {isSetParameters=true}
   &ShapeLocation(
    Box / Plane Mode:ENiagara_BoxPlaneMode=::NewEnumerator1
    MaskSpawnGroup=false
    Offset=$Particles.PlaneOffset
    Offset Mode=::NewEnumerator0
    Override Seed=false
    Plane Size:Vector2f=<9 23>
    Rotation Angle Type:ENiagara_AngleInput=::NewEnumerator0
    Rotation Mode=::NewEnumerator2
    Shape Primitive=::NewEnumerator2
    UseOwnerScale=true
    Yaw / Pitch / Roll=$Particles.PlaneAngle
   )
   &AddVelocity(
    Module.Use Falloff Distance=false
    Override Seed=false
    Velocity Mode=::NewEnumerator1
    Velocity Speed=&FloatFromCurve(
     FloatCurve=DataInterface() {propertyValues {Curve {keys [
         {arriveTangent=13.4013, interpMode=RCIM_Cubic, leaveTangent=13.4013, tangentMode=RCTM_User, time=0.1, value=0}
         {interpMode=RCIM_Cubic, tangentMode=RCTM_User, time=0.27, value=1}
        ]}}}
     CurveIndex=&RandomRangeFloat(Minimum=0, Maximum=1)
     Scale Curve=&RandomRangeFloat(Minimum=4, Maximum=6)
    )
   )
  ]
  particleUpdateScript [
   &ParticleState()
   &Drag(
    Drag=&FloatFromCurve(
     FloatCurve=DataInterface() {propertyValues {Curve {keys [
         {arriveTangent=0.1577, interpMode=RCIM_Cubic, leaveTangent=0.1577, tangentMode=RCTM_User, time=0, value=0.125}
         {arriveTangent=1.6704, interpMode=RCIM_Cubic, leaveTangent=1.6704, tangentMode=RCTM_User, time=1, value=1}
        ]}}}
     Scale Curve=&RandomRangeFloat(Minimum=2.5, Maximum=4)
    )
    Use Linear Drag=true
    Use Rotational Drag=true
   )
   &CurlNoiseForce(
    Bias Noise=false
    Noise Strength=&FloatFromCurve(
     FloatCurve=DataInterface() {propertyValues {Curve {keys [
         {arriveTangent=1.2297, interpMode=RCIM_Cubic, leaveTangent=1.2297, tangentMode=RCTM_User, time=0.0812, value=0}
         {interpMode=RCIM_Cubic, tangentMode=RCTM_User, time=0.7501, value=1}
        ]}}}
     Scale Curve=30
    )
    Pan Noise=true
    Pan Noise Field=&RandomRangeVector(Minimum=<0 0 0>, Maximum=<1 1 1>)
    Random Seed=33
   ) {module=CurlNoiseForce001}
   &GravityForce(Gravity=&RandomRangeVector(Minimum=<0 0 -11.25>, Maximum=<0 0 -16.875>))
   &SolveForcesAndVelocity(Clamp Velocity=false, Limit Acceleration=false)
   &ScaleColor(
    Linear Color Curve=DataInterface() {propertyValues {
      AlphaCurve {keys [<at 0 0>,<at 0.4658 0.35 RCIM_Cubic>,<at 0.9968 0>]}
      BlueCurve {keys [<at 0.0028 3>, {arriveTangent=-2.2061, interpMode=RCIM_Cubic, leaveTangent=-2.2061, time=0.2108, value=3}, <at 1 0.8 RCIM_Cubic>]}
      GreenCurve {keys [<at 0.0028 2.658>, {arriveTangent=-1.9091, interpMode=RCIM_Cubic, leaveTangent=-1.9091, time=0.2108, value=2.658}, <at 1 0.7542 RCIM_Cubic>]}
      RedCurve {keys [<at 0.0028 2.43>, {arriveTangent=-1.7264, interpMode=RCIM_Cubic, leaveTangent=-1.7264, time=0.2108, value=2.43}, <at 1 0.7083 RCIM_Cubic>]}
     }}
    Scale Mode=::NewEnumerator2
   )
   &ScaleColor(
    Color Value To Scale=$Particles.Color
    Scale Alpha=&Lerp_Float(
     A=1
     B=1
     Alpha=&RemapRange(
      Input Value To Remap=&VectorLength(Vector=$Engine.Owner.Velocity)
      Input Min=0
      Input Max=200
      Output Min=1
      Output Max=0
     )
    )
    ScaleA=true
    ScaleRGB=true
   ) {module=ScaleColor002}
   &ScaleSpriteSize(
    Uniform Curve Scale=&RandomRangeFloat(Minimum=3, Maximum=3.4)
    Uniform Curve Sprite Scale=DataInterface() {propertyValues {Curve {keys [
        {tangentMode=RCTM_User, time=0, value=1}
        {arriveTangent=-0.0405, interpMode=RCIM_Cubic, leaveTangent=-0.0405, tangentMode=RCTM_User, time=0.7009, value=1}
        {arriveTangent=-8.0089, interpMode=RCIM_Cubic, leaveTangent=-8.0089, tangentMode=RCTM_User, time=1, value=0}
       ]}}}
   )
   /SetVariables(Particles.SpriteSizeGlow:Vector2f=&Multiply_Vector2D_ByFloat(Float=1.2, Vector2D=$Particles.SpriteSize)) {isSetParameters=true}
   /SetVariables(Particles.Glow Color=&MultiplyLinearColor(Linear Color=$Particles.Color, Scale Factor=<2 2 2 0.15>)) {isSetParameters=true}
  ]
  rendererClasses [&NiagaraSpriteRendererProperties]
  renderers [
   {rendererClass=&NiagaraSpriteRendererProperties}
   {rendererClass=&NiagaraSpriteRendererProperties, rendererIndex=1}
  ]
 }
 Mist {
  emitterUpdateScript [
   &EmitterState()
   &SpawnRate(
    SpawnRate=&MakeCustomFloatFromBool(Bool=$System.IsRunning, True Float=16, False Float=6)
    Use Spawn Probability=false
   )
  ]
  particleSpawnScript [
   &InitializeParticle(
    Color=<rgba 0.94 0.9483 1 0.04>
    Lifetime=&RandomRangeFloat(Minimum=2.25, Maximum=3.75)
    Sprite Rotation Angle=&RandomRangeFloat(Minimum=0, Maximum=360)
    Sprite Rotation Mode=::NewEnumerator0
    Sprite Size Mode=::NewEnumerator3
    Sprite UV Mode=::NewEnumerator1
    Uniform Sprite Size=&RandomRangeFloat(Minimum=14, Maximum=26)
    UsePositionOffset=false
    Write Material Random=true
   )
   /SetVariables(Particles.faceindex:NiagaraInt32=&MakeIntFromFloat(
     Float=&FloatFromCurve(
      FloatCurve=DataInterface() {propertyValues {Curve {keys [<at 0 0>,<at 0.5967 0.5>,<at 0.8472 0.75>,<at 1 1>]}}}
      CurveIndex=&RandomRangeFloat(Minimum=0, Maximum=1)
      Scale Curve=4
     )
     Conversion Mode=::NewEnumerator3
    )) {isSetParameters=true}
   /SetVariables(
    Particles.PlaneAngle:Vector3f=&SelectVectorFromArray(
     Direct Array Index=$Particles.faceindex
     Vector Selection Array=DataInterface() {propertyValues {FloatData [<0 0 90>, <0 0 90>, <90 0 90>, <0 0 0>]}}
    )
    Particles.PlaneCamDepth:NiagaraFloat=&SelectFloatFromArray(Direct Array Index=$Particles.faceindex, Float Selection Array=DataInterface() {propertyValues {FloatData [5,12,3,7]}})
    Particles.PlaneGravity:NiagaraFloat=&MakeCustomFloatFromBool(
     Bool=&SetBoolByIntComparison(A=$Particles.faceindex, B=3, Comparison Type=::NewEnumerator4)
     True Float=0
     False Float=1
    )
    Particles.PlaneOffset:Vector3f=&SelectVectorFromArray(
     Direct Array Index=$Particles.faceindex
     Vector Selection Array=DataInterface() {propertyValues {FloatData [<0 -11 10>, <0 -11.5 -7.5>, <5.5 -7 4>, <0 -6 13.5>]}}
    )
    Particles.PlaneScale:Vector2f=&SelectVector2DFromArray(
     Direct Array Index=$Particles.faceindex
     Vector 2D Selection Array=DataInterface() {propertyValues {FloatData [<7 7>, <11 5>, <4 16>, <5 4>]}}
    )
   ) {isSetParameters=true}
   &ShapeLocation(
    Box / Plane Mode:ENiagara_BoxPlaneMode=::NewEnumerator1
    MaskSpawnGroup=false
    Offset=$Particles.PlaneOffset
    Offset Coordinate Space=::Local
    Offset Mode=::NewEnumerator0
    Override Seed=false
    Plane Size:Vector2f=$Particles.PlaneScale
    Rotation Angle Type:ENiagara_AngleInput=::NewEnumerator0
    Rotation Mode=::NewEnumerator2
    Shape Primitive=::NewEnumerator2
    UseOwnerScale=true
    Yaw / Pitch / Roll=$Particles.PlaneAngle
   )
   /SetVariables(Particles.LocalInitPos:NiagaraPosition=&TransformPosition(Position=$Particles.Position, Source Space=::World, Destination Space=::Local)) {isSetParameters=true}
  ]
  particleUpdateScript [
   &ParticleState()
   &ScaleColor(
    Scale Alpha=&FloatFromCurve(FloatCurve=DataInterface() {propertyValues {Curve {keys [
         {arriveTangent=9.5626, interpMode=RCIM_Cubic, leaveTangent=9.5626, tangentMode=RCTM_User, time=0, value=0}
         {arriveTangent=2.3621, interpMode=RCIM_Cubic, leaveTangent=2.3621, tangentMode=RCTM_User, time=0.1035, value=0.6155}
         {arriveTangent=0.0064, interpMode=RCIM_Cubic, leaveTangent=0.0064, time=0.4622, value=0.9143}
         {arriveTangent=-3.234, interpMode=RCIM_Cubic, leaveTangent=-3.234, tangentMode=RCTM_User, time=0.7475, value=0.6196}
         <at 1 0 RCIM_Cubic>
        ]}}})
    ScaleA=true
    ScaleRGB=true
   )
   &ScaleColor(
    Color Value To Scale=$Particles.Color
    Scale Mode=::NewEnumerator1
    Scale RGBA=&MakeVector4_FromLinearColor(LinearColor=&FNTimeOfDayColor(
      Night Color=<rgba 0.539 0.6941 1 1>
      Morning Color=<rgba 0.8743 0.9172 1 1>
      Day Color=<rgba 1 1 1 1>
      Evening Color=<rgba 0.8522 0.865 1 0.8>
     ))
    ScaleRGBA=true
   ) {module=ScaleColor002}
   &GravityForce(Gravity=&Multiply_VectorByFloat(
     Vector=&RandomRangeVector(Minimum=<0 0 -46.6667>, Maximum=<0 0 -66.6667>)
     Float=&FloatFromCurve(
      FloatCurve=DataInterface() {propertyValues {Curve {
         defaultValue=1
         keys [
          {arriveTangent=-0.4429, interpMode=RCIM_Cubic, leaveTangent=-0.4429, tangentMode=RCTM_User, time=0, value=1}
          {arriveTangent=0.0224, interpMode=RCIM_Cubic, leaveTangent=0.0224, tangentMode=RCTM_User, time=0.5, value=0.25}
         ]
        }}}
      Scale Curve=$Particles.Initial.PlaneGravity
     )
    ))
   &Drag(Drag=5, Use Linear Drag=true, Use Rotational Drag=true)
   &InheritSourceMovement(
    Applied Position Falloff Distance=30
    Applied Position Scale=&VectorFromFloat(Value=&Lerp_Float(
      A=1
      B=&FloatFromCurve(FloatCurve=DataInterface() {propertyValues {Curve {keys [
           <at 0.25 1 RCIM_Cubic>
           {
            arriveTangent=-0.507
            arriveTangentWeight=0.2296
            interpMode=RCIM_Cubic
            leaveTangent=-0.507
            leaveTangentWeight=0.1236
            tangentMode=RCTM_User
            time=0.6
            value=0.94
           }
           <at 1 0.75>
          ]}}})
      Alpha=&FloatFromCurve(
       FloatCurve=DataInterface() {propertyValues {Curve {defaultValue=1, keys [<at 0.0625 0.7>,<at 0.3125 1>,<at 0.625 1>,<at 1 1>]}}}
       CurveIndex=&Divide_Float(A=&VectorLength(Vector=$Engine.Owner.Velocity), B=800)
      )
     ))
    ApplyForce=false
    ApplyPosition=true
    ApplyPositionFalloff=false
   )
   /SetVariables(Particles.Age:NiagaraFloat=&Lerp_Float(
     A=$Particles.Age
     B=&Add_Float(
      A=&Multiply_Float(A=0.1, B=&Multiply_Float(A=$Engine.DeltaTime, B=$Particles.Initial.Lifetime))
      B=$Particles.Age
     )
     Alpha=&ClampFloat(Float=&RemapRange(
       Input Value To Remap=&VectorLength(Vector=$Engine.Owner.Velocity)
       Input Min=400
       Input Max=600
       Output Min=0
       Output Max=1
      ))
    )) {isSetParameters=true}
   &SolveForcesAndVelocity(Clamp Velocity=false, Limit Acceleration=false)
   &ScaleSpriteSize()
   /SetVariables(Particles.SubImageIndex=&FloatFromCurve(FloatCurve=DataInterface() {propertyValues {Curve {keys [
         {arriveTangent=106.1804, interpMode=RCIM_Cubic, leaveTangent=106.1804, tangentMode=RCTM_User, time=0, value=0}
         {arriveTangent=32.8034, interpMode=RCIM_Cubic, leaveTangent=32.8034, tangentMode=RCTM_User, time=1, value=63.9}
        ]}}})) {isSetParameters=true}
   &CameraOffset(Camera Offset Amount=&Multiply_Float(A=$Particles.Initial.PlaneCamDepth, B=&FN_FrontEndFloatModifier(Gameplay=1, Front End=1.3))) {module=CameraOffset001}
   &DynamicMaterialParameters(
    Index 0 Param 1=3
    Index 0 Param 2=1
    Index 0 Param 3=0.5
    Module.Index0Float0Write=true
    Module.Index0Float1Write=false
    Module.Index0Float2Write=false
    Module.Index0Float3Write=false
   )
  ]
  rendererClasses [&NiagaraSpriteRendererProperties]
  renderers [{rendererClass=&NiagaraSpriteRendererProperties}]
 }
}
summary {
 systemName=NS_Backpack_RevoltCrush
 emitters [
  {emitterName=Mist, rendererClasses [&NiagaraSpriteRendererProperties]}
  {emitterName=Embers, rendererClasses [&NiagaraSpriteRendererProperties]}
 ]
 userVariables [
  User.bIsFrontEnd {type=NiagaraBool}
  User.bIsFrontEndPreview {type=NiagaraBool}
  User.bisSwinging {type=NiagaraBool}
  User.IsSkydiving {type=NiagaraBool}
  User.OwnerVelocity {type=NiagaraFloat, defaultValue=0, description=""}
 ]
}
