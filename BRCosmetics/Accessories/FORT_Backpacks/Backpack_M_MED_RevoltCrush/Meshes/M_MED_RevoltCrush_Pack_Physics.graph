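# Physics asset for the RevoltCrush backpack cosmetic, defining bodies and constraints for its two dangling cords, the pack and a set of spine bones.
# Front and back cords are chains of small capsules rooted to the pack box through master bodies, mostly with limited swing and very tight twist.
# Large capsules on four spine bones form a torso-like volume, each constrained directly to the pack with free swing and twist.
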PhysicsAsset /BRCosmetics/Accessories/FORT_Backpacks/Backpack_M_MED_RevoltCrush/Meshes/M_MED_RevoltCrush_Pack_Physics
bodies {
 RevoltCrush_spine_01 [{shapeType=Capsule, length=7.9438, radius=13.5586, center=<-1.7965 0.1884 0>}]
 RevoltCrush_spine_02 [
  {shapeType=Capsule, length=7.3264, radius=13.6322, center=<-1.711 -0.1851 0.0001>}
  {shapeType=Capsule, length=29.3956, radius=11.5807, center=<2.5024 0.4992 3.9825>, rotation=<rot 90 0.0049 0.0049>}
  {shapeType=Capsule, length=29.875, radius=11.1013, center=<2.4704 0.4991 -3.9825>, rotation=<rot 90 -0.0396 -0.0396>}
 ]
 RevoltCrush_spine_03 [{shapeType=Capsule, length=8.2107, radius=10, center=<0.1445 0.2396 0.0001>}]
 RevoltCrush_spine_04 [{shapeType=Capsule, length=9.4404, radius=10, center=<2.766 0.7765 0.0002>}]
 dyn_cord_bck_01 [{shapeType=Capsule, length=1.5, radius=1, shapeName=dyn_cord_bck_01_capsule, center=<1 0 0>, rotation=<rot 90 0 0>}]
 dyn_cord_bck_02 [{shapeType=Capsule, length=1.5, radius=1, shapeName=dyn_cord_bck_02_capsule, center=<1 0 0>, rotation=<rot 90 0 0>}]
 dyn_cord_bck_03 [{shapeType=Capsule, length=1.5, radius=1, shapeName=dyn_cord_bck_03_capsule, center=<1 0 0>, rotation=<rot 90 0 0>}]
 dyn_cord_bck_04 [{shapeType=Capsule, length=1.5, radius=1, shapeName=dyn_cord_bck_04_capsule, center=<1 0 0>, rotation=<rot 90 0 0>}]
 dyn_cord_bck_05 [{shapeType=Capsule, length=1.5, radius=1, shapeName=dyn_cord_bck_05_capsule, center=<1 0 0>, rotation=<rot 90 0 0>}]
 dyn_cord_bck_06 [{shapeType=Capsule, length=1.5, radius=1, shapeName=dyn_cord_bck_06_capsule, center=<1 0 0>, rotation=<rot 90 0 0>}]
 dyn_cord_bck_MSTR [{shapeType=Capsule, length=0.1, radius=0.2, shapeName=dyn_cord_bck_MSTR_capsule, rotation=<rot 0 -90 0>}]
 dyn_cord_fr_01 [{shapeType=Capsule, length=1.5, radius=1, shapeName=dyn_cord_fr_01_capsule, center=<1.5 0 0>, rotation=<rot 90 0 0>}]
 dyn_cord_fr_02 [{shapeType=Capsule, length=1.5, radius=1, shapeName=dyn_cord_fr_02_capsule, center=<1.5 0 0>, rotation=<rot 90 0 0>}]
 dyn_cord_fr_03 [{shapeType=Capsule, length=1.5, radius=1, shapeName=dyn_cord_fr_03_capsule, center=<1.5 0 0>, rotation=<rot 90 0 0>}]
 dyn_cord_fr_04 [{shapeType=Capsule, length=1.5, radius=1, shapeName=dyn_cord_fr_04_capsule, center=<1.5 0 0>, rotation=<rot 90 0 0>}]
 dyn_cord_fr_05 [{shapeType=Capsule, length=1.5, radius=1, shapeName=dyn_cord_fr_05_capsule, center=<1.5 0 0>, rotation=<rot 90 0 0>}]
 dyn_cord_fr_MSTR [{shapeType=Capsule, length=0.1, radius=0.2, shapeName=dyn_cord_fr_MSTR_capsule, rotation=<rot 0 -90 0>}]
 pack [{shapeType=Box, extentX=13.5, extentY=11.4972, extentZ=14.9973, shapeName=pack_box, center=<-0.4272 -5.859 -10.3908>}]
}
constraints [
 {bone1Name=dyn_cord_bck_MSTR, bone2Name=pack, swing1LimitDegrees=45, swing2LimitDegrees=45, twistLimitDegrees=45}
 {bone1Name=dyn_cord_bck_01, bone2Name=dyn_cord_bck_MSTR, swing1LimitDegrees=30, swing2LimitDegrees=30, twistLimitDegrees=2}
 {bone1Name=dyn_cord_bck_02, bone2Name=dyn_cord_bck_01, swing1LimitDegrees=30, swing2LimitDegrees=30, twistLimitDegrees=2}
 {bone1Name=dyn_cord_bck_03, bone2Name=dyn_cord_bck_02, swing1LimitDegrees=30, swing2LimitDegrees=30, twistLimitDegrees=2}
 {bone1Name=dyn_cord_bck_04, bone2Name=dyn_cord_bck_03, swing1LimitDegrees=30, swing2LimitDegrees=30, twistLimitDegrees=2}
 {bone1Name=dyn_cord_bck_05, bone2Name=dyn_cord_bck_04, swing1LimitDegrees=30, swing2LimitDegrees=30, twistLimitDegrees=2}
 {bone1Name=dyn_cord_bck_06, bone2Name=dyn_cord_bck_05, swing1LimitDegrees=30, swing2LimitDegrees=30, twistLimitDegrees=2}
 {bone1Name=dyn_cord_fr_MSTR, bone2Name=pack, swing1LimitDegrees=30, swing2LimitDegrees=30, twistLimitDegrees=2}
 {bone1Name=dyn_cord_fr_01, bone2Name=dyn_cord_fr_MSTR, swing1LimitDegrees=30, swing2LimitDegrees=30, twistLimitDegrees=2}
 {bone1Name=dyn_cord_fr_02, bone2Name=dyn_cord_fr_01, swing1LimitDegrees=30, swing2LimitDegrees=30, twistLimitDegrees=2}
 {bone1Name=dyn_cord_fr_03, bone2Name=dyn_cord_fr_02, swing1LimitDegrees=30, swing2LimitDegrees=30, twistLimitDegrees=2}
 {bone1Name=dyn_cord_fr_04, bone2Name=dyn_cord_fr_03, swing1LimitDegrees=30, swing2LimitDegrees=30, twistLimitDegrees=2}
 {bone1Name=dyn_cord_fr_05, bone2Name=dyn_cord_fr_04, swing1LimitDegrees=30, swing2LimitDegrees=30, twistLimitDegrees=2}
 {
  bone1Name=RevoltCrush_spine_01
  bone2Name=pack
  swing1LimitDegrees=45
  swing1Motion=Free
  swing2LimitDegrees=45
  swing2Motion=Free
  twistLimitDegrees=45
  twistMotion=Free
 }
 {
  bone1Name=RevoltCrush_spine_02
  bone2Name=pack
  swing1LimitDegrees=45
  swing1Motion=Free
  swing2LimitDegrees=45
  swing2Motion=Free
  twistLimitDegrees=45
  twistMotion=Free
 }
 {
  bone1Name=RevoltCrush_spine_03
  bone2Name=pack
  swing1LimitDegrees=45
  swing1Motion=Free
  swing2LimitDegrees=45
  swing2Motion=Free
  twistLimitDegrees=45
  twistMotion=Free
 }
 {
  bone1Name=RevoltCrush_spine_04
  bone2Name=pack
  swing1LimitDegrees=45
  swing1Motion=Free
  swing2LimitDegrees=45
  swing2Motion=Free
  twistLimitDegrees=45
  twistMotion=Free
 }
]
