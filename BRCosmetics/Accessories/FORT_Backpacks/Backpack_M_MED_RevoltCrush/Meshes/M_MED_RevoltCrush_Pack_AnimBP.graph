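# Animation Blueprint for the RevoltCrush backpack cosmetic, adding secondary motion to its cords, handle and ring.
# A handle-holder spring and bone modifiers for spine, pack, handle and ring build a cached pose.
# That pose feeds RigidBody or chained AnimDynamics cord simulation, selected by bRigidBodyEnabled.
# The event graph samples body spine transforms, derives pack offsets and sine-plus-noise vibration, sets skydive gravity, and flags eliminations for five seconds.
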AnimBlueprint /BRCosmetics/Accessories/FORT_Backpacks/Backpack_M_MED_RevoltCrush/Meshes/M_MED_RevoltCrush_Pack_AnimBP
target_skeleton=&/M_MED_RevoltCrush_Pack_Skeleton
graphs [
 AnimGraph {nodes [
   n1 = Anim Dynamics - Bone: dyn_cord_fr_01!(
    ComponentPose=Local To Component(LocalPose=Use cached pose 'PreSimPose'())
    Alpha=1
    GravityOverride:Vector="0,0,-980"
    ExternalForce="0,0,0"
   ) -> (Pose) {
    angularDampingOverride=0.9998
    angularSpringConstant=50
    bAngularSpring=true
    bOverrideAngularDamping=true
    bUseGravityOverride=true
    bUsePlanarLimit=false
    simulationSpace=BoneRelative
    alphaBoolBlend {blendInTime=0.5, blendOption=QuarticInOut, blendOutTime=0.5}
    boundBone {boneName=dyn_cord_fr_01}
    componentAppliedLinearAccClamp=<350 350 350>
    componentLinearAccScale=<0.4 0.4 0.4>
    componentLinearVelScale=<0.5 0.5 0.5>
    gravityOverride {z=-980}
    physicsBodyDefinitions [{
      boundBone {boneName=dyn_cord_fr_01}
      boxExtents=<10 10 10>
      constraintSetup {angularLimitsMax=<1 30 0>, angularLimitsMin=<-1 -10 -30>, angularTarget=<1 0 0>}
      localJointOffset=<1 0 0>
     }]
    planarLimits [{drivingBone {boneName=pack}, planeTransform {location=<-20 10 0>, rotation=<rot 90 90 0>}}]
    relativeSpaceBone {boneName=pack}
   }
   n2 = Anim Dynamics - Bone: dyn_cord_fr_02!(ComponentPose=@n1, Alpha=1, GravityOverride:Vector="0,0,-980", ExternalForce="0,0,0") -> (Pose) {
    angularDampingOverride=0.9998
    angularSpringConstant=50
    bAngularSpring=true
    bOverrideAngularDamping=true
    bUseGravityOverride=true
    bUsePlanarLimit=false
    simulationSpace=BoneRelative
    alphaBoolBlend {blendInTime=0.5, blendOption=QuarticInOut, blendOutTime=0.5}
    boundBone {boneName=dyn_cord_fr_02}
    componentAppliedLinearAccClamp=<350 350 350>
    componentLinearAccScale=<0.4 0.4 0.4>
    componentLinearVelScale=<0.5 0.5 0.5>
    gravityOverride {z=-980}
    physicsBodyDefinitions [{
      boundBone {boneName=dyn_cord_fr_02}
      boxExtents=<10 10 10>
      constraintSetup {angularLimitsMax=<1 30 0>, angularLimitsMin=<-1 -10 -30>, angularTarget=<1 0 0>}
      localJointOffset=<1.5 0 0>
     }]
    planarLimits [{drivingBone {boneName=pack}, planeTransform {location=<-20 10 0>, rotation=<rot 90 90 0>}}]
    relativeSpaceBone {boneName=pack}
   }
   n3 = Anim Dynamics - Bone: dyn_cord_fr_03!(ComponentPose=@n2, Alpha=1, GravityOverride:Vector="0,0,-980", ExternalForce="0,0,0") -> (Pose) {
    angularDampingOverride=0.9998
    angularSpringConstant=50
    bAngularSpring=true
    bOverrideAngularDamping=true
    bUseGravityOverride=true
    bUsePlanarLimit=false
    simulationSpace=BoneRelative
    alphaBoolBlend {blendInTime=0.5, blendOption=QuarticInOut, blendOutTime=0.5}
    boundBone {boneName=dyn_cord_fr_03}
    componentAppliedLinearAccClamp=<350 350 350>
    componentLinearAccScale=<0.4 0.4 0.4>
    componentLinearVelScale=<0.5 0.5 0.5>
    gravityOverride {z=-980}
    physicsBodyDefinitions [{
      boundBone {boneName=dyn_cord_fr_03}
      boxExtents=<10 10 10>
      constraintSetup {angularLimitsMax=<1 30 0>, angularLimitsMin=<-1 -10 -30>, angularTarget=<1 0 0>}
      localJointOffset=<1.5 0 0>
     }]
    planarLimits [{drivingBone {boneName=pack}, planeTransform {location=<-20 10 0>, rotation=<rot 90 90 0>}}]
    relativeSpaceBone {boneName=pack}
   }
   n4 = Anim Dynamics - Bone: dyn_cord_fr_04!(ComponentPose=@n3, Alpha=1, GravityOverride:Vector="0,0,-980", ExternalForce="0,0,0") -> (Pose) {
    angularDampingOverride=0.9998
    angularSpringConstant=50
    bAngularSpring=true
    bOverrideAngularDamping=true
    bUseGravityOverride=true
    bUsePlanarLimit=false
    simulationSpace=BoneRelative
    alphaBoolBlend {blendInTime=0.5, blendOption=QuarticInOut, blendOutTime=0.5}
    boundBone {boneName=dyn_cord_fr_04}
    componentAppliedLinearAccClamp=<350 350 350>
    componentLinearAccScale=<0.4 0.4 0.4>
    componentLinearVelScale=<0.5 0.5 0.5>
    gravityOverride {z=-980}
    physicsBodyDefinitions [{
      boundBone {boneName=dyn_cord_fr_04}
      boxExtents=<10 10 10>
      constraintSetup {angularLimitsMax=<1 30 10>, angularLimitsMin=<-1 -10 -30>, angularTarget=<1 0 0>}
      localJointOffset=<2 0 0>
     }]
    planarLimits [{drivingBone {boneName=pack}, planeTransform {location=<-20 10 0>, rotation=<rot 90 90 0>}}]
    relativeSpaceBone {boneName=pack}
   }
   n5 = Anim Dynamics - Bone: dyn_cord_fr_05!(ComponentPose=@n4, Alpha=1, GravityOverride:Vector="0,0,-980", ExternalForce="0,0,0") -> (Pose) {
    angularDampingOverride=0.9998
    angularSpringConstant=50
    bAngularSpring=true
    bOverrideAngularDamping=true
    bUseGravityOverride=true
    bUsePlanarLimit=false
    simulationSpace=BoneRelative
    alphaBoolBlend {blendInTime=0.5, blendOption=QuarticInOut, blendOutTime=0.5}
    boundBone {boneName=dyn_cord_fr_05}
    componentAppliedLinearAccClamp=<350 350 350>
    componentLinearAccScale=<0.4 0.4 0.4>
    componentLinearVelScale=<0.5 0.5 0.5>
    gravityOverride {z=-980}
    physicsBodyDefinitions [{
      boundBone {boneName=dyn_cord_fr_05}
      boxExtents=<10 10 10>
      constraintSetup {angularLimitsMax=<1 30 10>, angularLimitsMin=<-1 -10 -30>, angularTarget=<1 0 0>}
      localJointOffset=<2 0 0>
     }]
    planarLimits [{drivingBone {boneName=pack}, planeTransform {location=<-20 10 0>, rotation=<rot 90 90 0>}}]
    relativeSpaceBone {boneName=pack}
   }
   n6 = Anim Dynamics - Bone: dyn_cord_bck_01!(ComponentPose=@n5, Alpha=1, GravityOverride:Vector="0,0,-980", ExternalForce="0,0,0") -> (Pose) {
    angularDampingOverride=0.9998
    angularSpringConstant=50
    bAngularSpring=true
    bOverrideAngularDamping=true
    bUseGravityOverride=true
    bUsePlanarLimit=false
    simulationSpace=BoneRelative
    alphaBoolBlend {blendInTime=0.5, blendOption=QuarticInOut, blendOutTime=0.5}
    boundBone {boneName=dyn_cord_bck_01}
    componentAppliedLinearAccClamp=<350 350 350>
    componentLinearAccScale=<0.4 0.4 0.4>
    componentLinearVelScale=<0.5 0.5 0.5>
    gravityOverride {z=-980}
    physicsBodyDefinitions [{
      boundBone {boneName=dyn_cord_bck_01}
      boxExtents=<10 10 10>
      constraintSetup {angularLimitsMax=<1 30 0>, angularLimitsMin=<-1 -10 -30>, angularTarget=<1 0 0>}
      localJointOffset=<1 0 0>
     }]
    planarLimits [{drivingBone {boneName=pack}, planeTransform {location=<-20 10 0>, rotation=<rot 90 90 0>}}]
    relativeSpaceBone {boneName=pack}
   }
   n7 = Anim Dynamics - Bone: dyn_cord_bck_02!(ComponentPose=@n6, Alpha=1, GravityOverride:Vector="0,0,-980", ExternalForce="0,0,0") -> (Pose) {
    angularDampingOverride=0.9998
    angularSpringConstant=50
    bAngularSpring=true
    bOverrideAngularDamping=true
    bUseGravityOverride=true
    bUsePlanarLimit=false
    simulationSpace=BoneRelative
    alphaBoolBlend {blendInTime=0.5, blendOption=QuarticInOut, blendOutTime=0.5}
    boundBone {boneName=dyn_cord_bck_02}
    componentAppliedLinearAccClamp=<350 350 350>
    componentLinearAccScale=<0.4 0.4 0.4>
    componentLinearVelScale=<0.5 0.5 0.5>
    gravityOverride {z=-980}
    physicsBodyDefinitions [{
      boundBone {boneName=dyn_cord_bck_02}
      boxExtents=<10 10 10>
      constraintSetup {angularLimitsMax=<1 30 0>, angularLimitsMin=<-1 -10 -30>, angularTarget=<1 0 0>}
      localJointOffset=<1.5 0 0>
     }]
    planarLimits [{drivingBone {boneName=pack}, planeTransform {location=<-20 10 0>, rotation=<rot 90 90 0>}}]
    relativeSpaceBone {boneName=pack}
   }
   n8 = Anim Dynamics - Bone: dyn_cord_bck_03!(ComponentPose=@n7, Alpha=1, GravityOverride:Vector="0,0,-980", ExternalForce="0,0,0") -> (Pose) {
    angularDampingOverride=0.9998
    angularSpringConstant=50
    bAngularSpring=true
    bOverrideAngularDamping=true
    bUseGravityOverride=true
    bUsePlanarLimit=false
    simulationSpace=BoneRelative
    alphaBoolBlend {blendInTime=0.5, blendOption=QuarticInOut, blendOutTime=0.5}
    boundBone {boneName=dyn_cord_bck_03}
    componentAppliedLinearAccClamp=<350 350 350>
    componentLinearAccScale=<0.4 0.4 0.4>
    componentLinearVelScale=<0.5 0.5 0.5>
    gravityOverride {z=-980}
    physicsBodyDefinitions [{
      boundBone {boneName=dyn_cord_bck_03}
      boxExtents=<10 10 10>
      constraintSetup {angularLimitsMax=<1 30 0>, angularLimitsMin=<-1 -10 -30>, angularTarget=<1 0 0>}
      localJointOffset=<1.5 0 0>
     }]
    planarLimits [{drivingBone {boneName=pack}, planeTransform {location=<-20 10 0>, rotation=<rot 90 90 0>}}]
    relativeSpaceBone {boneName=pack}
   }
   n9 = Anim Dynamics - Bone: dyn_cord_bck_04!(ComponentPose=@n8, Alpha=1, GravityOverride:Vector="0,0,-980", ExternalForce="0,0,0") -> (Pose) {
    angularDampingOverride=0.9998
    angularSpringConstant=50
    bAngularSpring=true
    bOverrideAngularDamping=true
    bUseGravityOverride=true
    bUsePlanarLimit=false
    simulationSpace=BoneRelative
    alphaBoolBlend {blendInTime=0.5, blendOption=QuarticInOut, blendOutTime=0.5}
    boundBone {boneName=dyn_cord_bck_04}
    componentAppliedLinearAccClamp=<350 350 350>
    componentLinearAccScale=<0.4 0.4 0.4>
    componentLinearVelScale=<0.5 0.5 0.5>
    gravityOverride {z=-980}
    physicsBodyDefinitions [{
      boundBone {boneName=dyn_cord_bck_04}
      boxExtents=<10 10 10>
      constraintSetup {angularLimitsMax=<1 30 0>, angularLimitsMin=<-1 -10 -30>, angularTarget=<1 0 0>}
      localJointOffset=<2 0 0>
     }]
    planarLimits [{drivingBone {boneName=pack}, planeTransform {location=<-20 10 0>, rotation=<rot 90 90 0>}}]
    relativeSpaceBone {boneName=pack}
   }
   n10 = Anim Dynamics - Bone: dyn_cord_bck_05!(ComponentPose=@n9, Alpha=1, GravityOverride:Vector="0,0,-980", ExternalForce="0,0,0") -> (Pose) {
    angularDampingOverride=0.9998
    angularSpringConstant=50
    bAngularSpring=true
    bOverrideAngularDamping=true
    bUseGravityOverride=true
    bUsePlanarLimit=false
    simulationSpace=BoneRelative
    alphaBoolBlend {blendInTime=0.5, blendOption=QuarticInOut, blendOutTime=0.5}
    boundBone {boneName=dyn_cord_bck_05}
    componentAppliedLinearAccClamp=<350 350 350>
    componentLinearAccScale=<0.4 0.4 0.4>
    componentLinearVelScale=<0.5 0.5 0.5>
    gravityOverride {z=-980}
    physicsBodyDefinitions [{
      boundBone {boneName=dyn_cord_bck_05}
      boxExtents=<10 10 10>
      constraintSetup {angularLimitsMax=<1 30 10>, angularLimitsMin=<-1 -10 -30>, angularTarget=<1 0 0>}
      localJointOffset=<2 0 0>
     }]
    planarLimits [{drivingBone {boneName=pack}, planeTransform {location=<-20 10 0>, rotation=<rot 90 90 0>}}]
    relativeSpaceBone {boneName=pack}
   }
   n11 = Anim Dynamics - Bone: dyn_cord_bck_06!(ComponentPose=@n10, Alpha=1, GravityOverride:Vector="0,0,-980", ExternalForce="0,0,0") -> (Pose) {
    angularDampingOverride=0.9998
    angularSpringConstant=50
    bAngularSpring=true
    bOverrideAngularDamping=true
    bUseGravityOverride=true
    bUsePlanarLimit=false
    simulationSpace=BoneRelative
    alphaBoolBlend {blendInTime=0.5, blendOption=QuarticInOut, blendOutTime=0.5}
    boundBone {boneName=dyn_cord_bck_06}
    componentAppliedLinearAccClamp=<350 350 350>
    componentLinearAccScale=<0.4 0.4 0.4>
    componentLinearVelScale=<0.5 0.5 0.5>
    gravityOverride {z=-980}
    physicsBodyDefinitions [{
      boundBone {boneName=dyn_cord_bck_06}
      boxExtents=<10 10 10>
      constraintSetup {angularLimitsMax=<1 30 10>, angularLimitsMin=<-1 -10 -30>, angularTarget=<1 0 0>}
      localJointOffset=<2 0 0>
     }]
    planarLimits [{drivingBone {boneName=pack}, planeTransform {location=<-20 10 0>, rotation=<rot 90 90 0>}}]
    relativeSpaceBone {boneName=pack}
   }
   n12 = RigidBody!(ComponentPose=Local To Component(LocalPose=Use cached pose 'PreSimPose'()), Alpha, OverrideWorldGravity:Vector="0,0,-980", ExternalForce) -> (Pose) {
    bOverrideWorldGravity=true
    overridePhysicsAsset=&/M_MED_RevoltCrush_Pack_Physics
    simulationSpace=BaseBoneSpace
    alphaBoolBlend {blendInTime=0.5, blendOption=QuarticInOut, blendOutTime=0.5}
    baseBoneRef {boneName=pack}
    overrideWorldGravity {z=-980}
    simSpaceSettings {
     dampingAlpha=0
     maxAngularAcceleration=10
     maxAngularVelocity=2
     maxLinearAcceleration=1100
     maxLinearVelocity=400
     worldAlpha=0.5
     externalLinearDragV=<0.5 0.5>
    }
   }
   Output Pose(Result=Blend Poses by bool(BlendPose_0=Component To Local(ComponentPose=@n12), BlendPose_1=Component To Local(ComponentPose=@n11), bActiveValue=Get bRigidBodyEnabled()) {blendType=Linear}) {layerGroup=None}
   n13 = Anim Dynamics - Bone: dyn_handle_holder(ComponentPose=Local To Component(LocalPose=Input Pose()), Alpha=1, ExternalForce="0,0,0") {
    angularDampingOverride=1
    angularSpringConstant=8000
    bAngularSpring=true
    bOverrideAngularDamping=true
    bUseGravityOverride=true
    alphaBoolBlend {blendOption=Linear}
    boundBone {boneName=dyn_handle_holder}
    componentAppliedLinearAccClamp=<2500 2500 5000>
    componentLinearAccScale=<0.5 0.5 0.5>
    componentLinearVelScale=<0.7 0.7 0.8>
    gravityOverride {z=-980}
    physicsBodyDefinitions [{
      boundBone {boneName=dyn_handle_holder}
      boxExtents=<20 20 15>
      constraintSetup {angularLimitsMax=<0 2 0>, angularLimitsMin=<0 -2 -15>, angularTarget=<1 0 0>}
      localJointOffset=<1 0 0>
     }]
   }
   n14 = "Transform (Modify) Bone - Bone: RevoltCrush_spine_01"!(ComponentPose=@n13, Alpha=1, Translation="0,0,0", Rotation="0,0,0") -> (Pose) {
    rotationMode=BMM_Replace
    rotationSpace=BCS_WorldSpace
    translationMode=BMM_Replace
    translationSpace=BCS_WorldSpace
    alphaBoolBlend {blendOption=Linear}
    boneToModify {boneName=RevoltCrush_spine_01}
   }
   n15 = "Transform (Modify) Bone - Bone: RevoltCrush_spine_02"!(ComponentPose=@n14, Alpha=1, Translation="0,0,0", Rotation="0,0,0") -> (Pose) {
    rotationMode=BMM_Replace
    rotationSpace=BCS_WorldSpace
    translationMode=BMM_Replace
    translationSpace=BCS_WorldSpace
    alphaBoolBlend {blendOption=Linear}
    boneToModify {boneName=RevoltCrush_spine_02}
   }
   n16 = "Transform (Modify) Bone - Bone: RevoltCrush_spine_03"!(ComponentPose=@n15, Alpha=1, Translation="0,0,0", Rotation="0,0,0") -> (Pose) {
    rotationMode=BMM_Replace
    rotationSpace=BCS_WorldSpace
    translationMode=BMM_Replace
    translationSpace=BCS_WorldSpace
    alphaBoolBlend {blendOption=Linear}
    boneToModify {boneName=RevoltCrush_spine_03}
   }
   n17 = "Transform (Modify) Bone - Bone: RevoltCrush_spine_04"!(ComponentPose=@n16, Alpha=1, Translation="0,0,0", Rotation="0,0,0") -> (Pose) {
    rotationMode=BMM_Replace
    rotationSpace=BCS_WorldSpace
    translationMode=BMM_Replace
    translationSpace=BCS_WorldSpace
    alphaBoolBlend {blendOption=Linear}
    boneToModify {boneName=RevoltCrush_spine_04}
   }
   n18 = "Transform (Modify) Bone - Bone: pack"!(ComponentPose=@n17, Alpha, Translation="0,2,0", Rotation="0,0,5") -> (Pose) {
    rotationMode=BMM_Additive
    rotationSpace=BCS_BoneSpace
    translationMode=BMM_Additive
    alphaBoolBlend {blendOption=Linear}
    boneToModify {boneName=pack}
    rotation {roll=5}
    translation {y=2}
   }
   n19 = "Transform (Modify) Bone - Bone: dyn_handle"!(ComponentPose=@n18, bAlphaBoolEnabled:bool=True, Rotation="0,0,0") -> (Pose) {alphaInputType=Bool, rotationMode=BMM_Additive, alphaBoolBlend {blendOption=Linear}, boneToModify {boneName=dyn_handle}}
   n20 = "Transform (Modify) Bone - Bone: dyn_ring"!(ComponentPose=@n19, bAlphaBoolEnabled:bool=True, Rotation="0,0,0") -> (Pose) {alphaInputType=Bool, rotationMode=BMM_Additive, alphaBoolBlend {blendOption=Linear}, boneToModify {boneName=dyn_ring}}
   Save cached pose 'PreSimPose'(Pose=Component To Local(ComponentPose=@n20)) {cacheName=PreSimPose}
  ]}
 CalculateVIbration {nodes [
   n1 = Inputs!() -> (Frequency:double=~, Amplitude:double=~, NoiseAmplitude:double=~)
   n2 = float * float(A=@n1.NoiseAmplitude, B=-1) {is=title K2Node_PromotableOperator}
   n3 = "Sin (Radians)"(A=float * float(A=@n1, B=GetGameTimeInSeconds()))
   Outputs!(Result:double=float + float(A=float * float(A=@n3, B=@n1.Amplitude), B=RandomFloatInRange(Min=@n2, Max=@n1.NoiseAmplitude)))
  ]}
 EventGraph {nodes [
   Event BlueprintUpdateAnimation()
   n1 = GetPartSkeletalMeshComponent()
   > Is Valid(InputObject=@n1)
   n2 = GetAttachParent!(self:Scene Component Object Reference=@n1) -> (ReturnValue:Scene Component Object Reference)
   > Is Valid(InputObject=@n2)
   > Set BodyMesh(BodyMesh:Scene Component Object Reference=@n2) -> (Output_Get:Scene Component Object Reference)
   > n3 = Sequence!(execute) -> (then_0, then_1, then_2:Exec, then_3:Exec)
   n4 = Branch(execute=@n3.then_1, Condition=Get bIsSkydiveFloating())
   > Set WorldGravity(WorldGravity:Vector="0,-100,600") -> (Output_Get:Vector)
   n5 = Branch(execute=@n4.else, Condition=Get bIsSkydiveDiveMode())
   > Set WorldGravity(WorldGravity="-500,0,700")
   Set WorldGravity(execute=@n5.else, WorldGravity="0,0,-980")
   n6 = Get BodyMesh() -> (BodyMesh:Scene Component Object Reference)
   n7 = BreakTransform(InTransform=GetSocketTransform(self=@n6, InSocketName=spine_01, TransformSpace=RTS_World))
   Set spine_01_location(execute=@n3.then_2, spine_01_location:Vector=@n7) -> (Output_Get:Vector)
   > Set spine_01_rotation(spine_01_rotation:Rotator=@n7.Rotation) -> (Output_Get:Rotator)
   n8 = BreakTransform(InTransform=GetSocketTransform(self=@n6, InSocketName=spine_02, TransformSpace=RTS_World))
   > Set spine_02_location(spine_02_location:Vector=@n8) -> (Output_Get:Vector)
   > Set spine_02_rotation(spine_02_rotation:Rotator=@n8.Rotation) -> (Output_Get:Rotator)
   n9 = BreakTransform(InTransform=GetSocketTransform(self=@n6, InSocketName=spine_03, TransformSpace=RTS_World))
   > Set spine_03_location(spine_03_location:Vector=@n9) -> (Output_Get:Vector)
   > Set spine_03_rotation(spine_03_rotation:Rotator=@n9.Rotation) -> (Output_Get:Rotator)
   n10 = BreakTransform(InTransform=GetSocketTransform(self=@n6, InSocketName=spine_04, TransformSpace=RTS_World))
   > Set spine_04_location(spine_04_location:Vector=@n10) -> (Output_Get:Vector)
   > Set spine_04_rotation(spine_04_rotation:Rotator=@n10.Rotation) -> (Output_Get:Rotator)
   Set HandleVibration(
    execute=@n3.then_0
    HandleVibration:Rotator=MakeRotator(Roll=Calculate VIbration!(Frequency:double=40, Amplitude:double=1, NoiseAmplitude:double=1) -> (Result:double=~) {is=title K2Node_MacroInstance})
   ) -> (Output_Get:Rotator)
   > Set RingVibration(RingVibration:Rotator=MakeRotator(Pitch=Calculate VIbration(Frequency=50, Amplitude=2, NoiseAmplitude=1) -> (Result=~))) -> (Output_Get:Rotator)
   > Set HolderHandleVibration(HolderHandleVibration:Rotator=MakeRotator(Roll=Calculate VIbration(Frequency=40, Amplitude=1, NoiseAmplitude=1) -> (Result=~))) -> (Output_Get:Rotator)
   n11 = GetOwnerSkeletalMeshComponent()
   n12 = BreakRotator(InRot=BreakTransform(InTransform=GetDeltaTransformFromRefPose(self=@n11, BoneName=spine_05, BaseName=spine_04)).Rotation)
   Set BackPack Rotation(execute=@n3.then_3, BackPack Rotation:Rotator=MakeRotator(Roll=MapRangeClamped(Value=@n12.Yaw, InRangeA=0, InRangeB=-30, OutRangeB=20))) -> (Output_Get:Rotator)
   > Set BackPack Translation(BackPack Translation:Vector=MakeVector(Y=MapRangeClamped(Value=@n12.Yaw, InRangeA=0, InRangeB=-30, OutRangeB=-2))) -> (Output_Get:Vector)
   n13 = Event BlueprintProcessAnimInputEvent()
   > Branch(Condition=IsNamed(self=@n13.AnimInputEvent, InEventName=aie.EnemyPlayerEliminated))
   > Set IsEliminated(IsEliminated:bool=true) -> (Output_Get:bool)
   > Delay(Duration=5)
   > Set IsEliminated(IsEliminated=false)
  ]}
]
variables [
 BackPack Rotation {type=Rotator}
 BackPack Translation {type=Vector}
 BodyMesh {type=SceneComponent, default=null}
 HandleVibration {type=Rotator}
 HolderHandleVibration {type=Rotator}
 IsEliminated {type=boolean}
 RingVibration {type=Rotator}
 spine_01_location {type=Vector}
 spine_01_rotation {type=Rotator}
 spine_02_location {type=Vector}
 spine_02_rotation {type=Rotator}
 spine_03_location {type=Vector}
 spine_03_rotation {type=Rotator}
 spine_04_location {type=Vector}
 spine_04_rotation {type=Rotator}
 WorldGravity {type=Vector}
]
